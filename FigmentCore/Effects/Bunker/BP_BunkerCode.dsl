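; BP_BunkerCode is a bunker effect actor that is visible only while a watched calendar event is active.
; On BeginPlay, where cosmetic events can execute, it queries calendar event information for its Event to Watch variable.
; It unhides the actor if that event is active and hides it otherwise.
; The check runs only once, at BeginPlay, and the construction script is empty.
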(event EventBeginPlay
  (Networking|CanExecuteCosmeticEvents
    (:True
      (bind _biseventactive (CalendarEvents|GetCalendarEventInformation (Variables|Default|GetEvent%20to%20Watch)))
      (if _biseventactive
        (Rendering|SetActorHiddenInGame false)
        (else
          (Rendering|SetActorHiddenInGame true))))))

(fn ConstructionScript)
