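# Additive, unlit, two-sided material for a blinking bunker light effect, with a tintable emissive color, dark red by default.
# Vertex red shapes the glow, and vertex green blends in panning world-aligned dust noise through the falloff exponent and adds camera-depth fade.
# Inverse fresnel softens grazing edges.
# A mask texture scrolled over time at MessageRate acts as a blink-pattern strip, modulating brightness above a minimum scale.
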Material /FigmentCore/Effects/Bunker/M_BunkerBlinker
graph {
 properties {bAutomaticallySetUsageInEditor=false, blendMode=BLEND_Additive, shadingModel=MSM_Unlit, twoSided=true, usedShadingModels=Unlit}
 nodes [
  n1 = VertexColor()
  n2 = WorldPosition()
  n3 = Multiply(A=Time(), B=VectorParameter() {parameterName=NoisePanSpeed, defaultValue=<rgba 2 5 -15 0>}.RGB)
  n4 = ComponentMask(None=MaterialFunctionCall!(
    Diffuse Texture Object=TextureObject() {texture=/Game/Global/Textures/Default/PointSplatter/T_Tiling_Dust}
    TextureSize=VectorParameter() {parameterName=NoiseSize, defaultValue=<rgba 250 250 400 0>}.RGB
    Texture Size Z projection
    Use seperate settings for Z projection
    Use seperate texture for Z projection
    Z Plane Diffuse Texture Object
    ProjectionTransitionContrast
    Use Texture Alpha Heightmap to Refine the Transistion
    HeightLerpTransitionContrast
    PickTheDiffuseHeightChannel
    TheDiffuseHeightChannel
    Export Float 4 Diffuse
    WorldPosition=Add(A=@n2.XYZ, B=@n3)
    World Space Normal
   ) {materialFunction=/Game/Environments/Helios/Props/Holiday/Holiday_LumpkinKing_A/Materials/Functions/MF_WorldAlignedTextures_Simple}.XY Texture) {g=true}
  n5 = LinearInterpolate(B=Max(A=LinearInterpolate(Alpha=@n4) {constA=20, constB=-20}) {constB=0.25}, Alpha=@n1.G) {constA=5}
  n6 = Fresnel() {baseReflectFraction=0}
  n7 = Time()
  n8 = ScalarParameter() {defaultValue=8, parameterName=MessageRate}
  n9 = AppendVector(A=Frac(None=Divide(A=@n7, B=@n8)), B=Constant() {r=0.5})
  n10 = TextureObjectParameter() {parameterName=Param, samplerType=SAMPLERTYPE_Masks, texture=/FigmentCore/Effects/Bunker/Texture/bunker_S06}
  n11 = Multiply(
   A=Multiply(A=Power(Base=@n1.R, Exp=@n5) {constExponent=10}, B=OneMinus(None=@n6))
   B=LinearInterpolate(
    A=ScalarParameter() {defaultValue=0.2, parameterName=MinBrightnessScale}
    Alpha=TextureSample!(UVs=@n9, Tex=@n10, Level, Apply View MipBias) {mipValueMode=TMVM_MipLevel, samplerType=SAMPLERTYPE_Masks, texture=/Engine/EngineResources/DefaultTexture}.R
   ) {constA=0.5}
  )
  n12 = ScalarParameter() {defaultValue=50, parameterName=Camera Fade Offset}
  LinearInterpolate(
   A=@n11
   B=Multiply(
    A=@n11
    B=MaterialFunctionCall(Fade Offset=@n12) {materialFunction=&CameraDepthFade}.Result
   )
   Alpha=@n1.G
  ) {constA=5}
  MaterialFunctionCall(Emissive=VectorParameter() {parameterName=Color, defaultValue=<rgba 0.1177 0 0.0036 1>}.RGB) {materialFunction=&MF_FX_EmissiveAdjustments}
 ]
}
