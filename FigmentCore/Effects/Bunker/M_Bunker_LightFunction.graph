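# Light function material for the Bunker effect that modulates light brightness over time in a repeating pattern read from a code texture.
# Time divided by MessageRate is wrapped with Frac to scroll U along a fixed middle row of CodeTexture, sampled at an explicit mip level.
# The red channel lerps brightness up from MinBrightnessScale, so the whole light flickers uniformly as a timed message rather than projecting a spatial pattern.
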Material /FigmentCore/Effects/Bunker/M_Bunker_LightFunction
graph {
 properties {bAutomaticallySetUsageInEditor=false, materialDomain=MD_LightFunction, usedShadingModels=Default Lit}
 nodes [
  n1 = Frac(None=Divide(A=Time(), B=ScalarParameter() {defaultValue=8, parameterName=MessageRate}))
  n2 = Constant() {r=0.5}
  LinearInterpolate(
   A=ScalarParameter() {defaultValue=0.2, parameterName=MinBrightnessScale}
   Alpha=TextureSample!(
    UVs=AppendVector(A=@n1, B=@n2)
    Tex=TextureObjectParameter() {parameterName=CodeTexture, samplerType=SAMPLERTYPE_Masks, texture=/FigmentCore/Effects/Bunker/Texture/bunker_S06}
    Level
    Apply View MipBias
   ) {mipValueMode=TMVM_MipLevel, samplerType=SAMPLERTYPE_Masks, texture=/Engine/EngineResources/DefaultTexture}.R
  ) {constA=0.5}
 ]
}
